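# Niagara system for cave water drips, falling as faint tinted, speed-stretched GPU sprites under user-set gravity.
# RainFromObjects spawns in a raised sphere, snaps particles to distance-field surfaces and the depth buffer, culls invalid or far-moved samples, then runs a ceiling/floor-pick scratch module.
# User parameters scale rate, radius, color and alpha.
# SplashesClose is disabled, retaining rain logic with time-of-day, indoor fading, hard-coded kill volumes and a NeonCity mask texture.
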NiagaraSystem /Game/Effects/Fort_Effects/Effects/Environment/Cave/NS_Cave_Drips
dependencies {
 usedDataInterfaces [
  &NiagaraDataInterfaceCollisionQuery
  &NiagaraDataInterfaceCurve
  &NiagaraDataInterfaceColorCurve
  /Script/Niagara.NiagaraDataInterfaceTexture
 ]
 usedDynamicInputs [&ColorFromCurve]
 usedModules [
  &SystemState
  &self:SystemSpawnScript.SystemScriptSource.SystemScriptGraph.NiagaraNodeAssignment_1.SetVariables_#1
 ]
}
emitters {
 RainFromObjects {
  simTarget=GPUComputeSim
  emitterUpdateScript [
   &EmitterState()
   &SpawnRate(SpawnRate=&Multiply_Float(A=800, B=$User.Spawn Rate Mult), Use Spawn Probability=false)
  ]
  particleSpawnScript [
   &InitializeParticle(
    Lifetime Mode=::NewEnumerator1
    Sprite Rotation Mode=::NewEnumerator1
    Sprite Size Max=<50 20>
    Sprite Size Min=<25 15>
    Sprite Size Mode=::NewEnumerator2
    UsePositionOffset=false
    Write Material Random=true
   ) {module=InitializeParticle001}
   &AddVelocity(Velocity=&RandomRangeVector(Minimum=<0 0 -50>, Maximum=<0 0 -50>))
   &SphereLocation(
    MaskSpawnGroup=false
    Non Uniform Scale=<1 1 0.7>
    Offset=<0 0 1500>
    Override Seed=false
    Sphere Radius=$User.Radius Scale
    Surface Only=false
    Surface Only Band Thickness=0.5
    UseNonUniformScale=true
    UseOffset=true
    UseSphereOrientation=false
   )
   /Niagara/Modules/Math/MoveToNearestDistanceFieldSurface_GPU(Kill by Distance=800, bKillByDistance=true)
   /SetVariables(Particles.InitialPos:NiagaraPosition=$Particles.Position) {isSetParameters=true}
   /Niagara/Modules/Math/PlaceParticlesOnDepthBuffer_GPU(Max Z Depth=1500, Module.Use Max Z Depth=true)
   &KillParticles(Kill Particles=&SetBoolByFloatComparison(
     A=&VectorLength(Vector=&Subtract_Vector(A=$Particles.InitialPos, B=Linked() {linkedVariable {name=Particles.Position, type=NiagaraPosition}}))
     B=1500
    ))
   &KillParticles(Kill Particles=&Boolean_NotOperation(Bool=$Output.PlaceParticlesOnDepthBuffer_GPU.ValidSample)) {module=KillParticles001}
   &SubUVAnimation(End Frame=3, SubUV Animation Mode:ENiagara_SubUVLookupMode=::NewEnumerator2)
   &CameraOffset(Camera Offset Amount=-10)
   &self:ScratchModule_02(Angle=180, BlendDirection=1, Normal:Vector3f=<0 0 1>, PickCeilingOrFloor=true, Velocity=-200) {module=ScratchModule_02}
  ]
  particleUpdateScript [
   &ParticleState()
   &RegionMask(Inner Float=0, Outer Float=2, Region Falloff=0, Region Radius:NiagaraFloat=$User.Rain_MaskInside_Radius, UseOffset=false, UseRotation=false) {module=RegionMask001}
   &ScaleSpriteSize(Uniform Curve Sprite Scale=DataInterface() {propertyValues {Curve {keys [{tangentMode=RCTM_User, time=0, value=0.85}, {tangentMode=RCTM_User, time=1, value=1}]}}})
   &GravityForce(Gravity=$User.Gravity Scale)
   &SolveForcesAndVelocity(Clamp Velocity=false, Limit Acceleration=false, Speed Limit=500)
   &ScaleSpriteSizeBySpeed(
    Max Scale Factor=<0.1 5>
    Min Scale Factor=<0.05 1>
    Source Velocity:Vector3f=&Multiply_Vector(A=<0.05 0.05 1>, B=$Particles.Velocity)
   )
   &ScaleColor(
    Scale Alpha=&Multiply_Float(
     A=&Multiply_Float(
      A=&Multiply_Float(
       A=&FloatFromCurve(FloatCurve=DataInterface() {propertyValues {Curve {keys [
            {interpMode=RCIM_Cubic, tangentMode=RCTM_User, time=0, value=0}
            {interpMode=RCIM_Cubic, tangentMode=RCTM_User, time=0.5, value=1}
            {tangentMode=RCTM_User, time=1, value=0}
           ]}}})
       B=1
      )
      B=1
     )
     B=$User.Rain_AlphaMult
    )
    Scale RGB=&MakeVectorFromLinearColorRGB(Color=&MultiplyLinearColor(Linear Color=$User.Rain_NonLitColor, Scale Factor=&MakeVector4_FromFloat(Float=$User.Rain_ColorMult)))
    ScaleA=true
    ScaleRGB=true
   )
  ]
  rendererClasses [&NiagaraSpriteRendererProperties]
  renderers [
   {rendererClass=&NiagaraSpriteRendererProperties}
   {rendererClass=&NiagaraSpriteRendererProperties, rendererIndex=1}
  ]
 }
 SplashesClose {
  bEnabled=false
  simTarget=GPUComputeSim
  emitterUpdateScript [
   &EmitterState(
    Enable Distance Culling=true
    MaxDistance=50000
    Module.bApplySpawnCountScaleByDistance=true
    Module.bUseMaxDistance=true
    Module.bUseMinDistance=false
    Scalability Mode=::NewEnumerator1
    Scale Spawn Count=true
    Spawn Count Scale By Distance=DataInterface() {propertyValues {Curve {keys [<at 0 1>,<at 0.25 1>,<at 1 0>]}}}
   )
   &SpawnRate(
    SpawnRate=&Multiply_Float(
     A=&FloatFromCurve(
      FloatCurve=DataInterface() {propertyValues {Curve {keys [<at 0.08 1>,<at 0.185 0>,<at 0.9375 0>,<at 1 1>]}}}
      CurveIndex=Linked() {linkedVariable {name=NPC.FortniteNPC.TimeOfDay_0-1, type=NiagaraFloat}}
      Scale Curve=$User.Rain_SpawnRateScale
     )
     B=&Multiply_Float(A=&Lerp_Float(A=1, B=0, Alpha=$NPC.FortniteNPC.IsPlayerIndoors), B=5000)
    )
    Use Spawn Probability=false
   )
  ]
  particleSpawnScript [
   &InitializeParticle(
    Lifetime=&RandomRangeFloat(Minimum=0.25, Maximum=0.35)
    Sprite Rotation Mode=::NewEnumerator2
    Sprite Size Mode=::NewEnumerator1
    Uniform Sprite Size Max=1
    Uniform Sprite Size Min=1
    UsePositionOffset=false
    Write Material Random=true
   )
   &AddVelocity(Velocity=&RandomRangeVector(Minimum=<0 0 25>, Maximum=<0 0 50>))
   &ShapeLocation(
    Cylinder Height=50
    Cylinder Radius=2000
    MaskSpawnGroup=false
    Offset=<0 0 -50>
    Offset Mode=::NewEnumerator0
    Override Seed=false
    Shape Primitive=::NewEnumerator1
    Surface Only=true
    Surface Only Band Thickness=0.9
    Use Endcaps In Surface Only Mode=false
    UseOwnerScale=true
   )
   /SetVariables(Particles.InitialPos:NiagaraPosition=$Particles.Position) {isSetParameters=true}
   &PlaceParticlesOnDepthBuffer_GPU(Max Z Depth=1, Module.Use Max Z Depth=true)
   &MoveToNearestDistanceFieldSurface_GPU(Kill by Distance=50, bKillByDistance=true)
   &KillParticles(Kill Particles=&Boolean_NotOperation(Bool=$Output.PlaceParticlesOnDepthBuffer_GPU.ValidSample)) {module=KillParticles001}
   /SetVariables(Particles.SpriteSize=&Vector2DFromFloat(Value=&RandomRangeFloat(
      Minimum=25
      Maximum=&UniformAOrBFloat(A=35, B=&RandomRangeFloat(Minimum=40, Maximum=50), Distribution Weight=0.1)
     ))) {isSetParameters=true}
   &CameraOffset(Camera Offset Amount=-25)
   &SubUVAnimation(End Frame=3, SubUV Animation Mode:ENiagara_SubUVLookupMode=::NewEnumerator2)
   &self:ScratchModule_05(Angle=10, BlendDirection=35, Normal:Vector3f=<0 0 1>, Velocity=5) {module=ScratchModule_05}
   &KillParticlesInVolume(
    Box Size:Vector3f=<3350 3825 675>
    Kill Shape=::NewEnumerator1
    Volume Origin:Vector3f=&AddVectorToPosition(Position=<-60714 21466 718>, Vector=<0 0 0>)
   )
   &KillParticlesInVolume(
    Box Size:Vector3f=<3350 2725 6825>
    Kill Shape=::NewEnumerator1
    Volume Origin:Vector3f=&AddVectorToPosition(Position=<-52356 19960 3232>, Vector=<0 0 0>)
   ) {module=KillParticlesInVolume002}
   &KillParticlesInVolume(
    Box Size:Vector3f=<3350 3000 2500>
    Kill Shape=::NewEnumerator1
    Volume Origin:Vector3f=&AddVectorToPosition(Position=<-60727 26090 612>, Vector=<0 0 0>)
   ) {module=KillParticlesInVolume004}
   &KillParticlesInVolume(
    Box Size:Vector3f=<3325 3000 6875>
    Kill Shape=::NewEnumerator1
    Volume Origin:Vector3f=&AddVectorToPosition(Position=<-55993 24781 1422>, Vector=<0 0 0>)
   ) {module=KillParticlesInVolume005}
   &KillParticlesInVolume(
    Box Size:Vector3f=<2050 2975 3125>
    Kill Shape=::NewEnumerator1
    Volume Origin:Vector3f=&AddVectorToPosition(Position=<-50898 23839 929>, Vector=<0 0 0>)
   ) {module=KillParticlesInVolume006}
   &KillParticlesInVolume(
    Box Size:Vector3f=<2050 2050 2625>
    Kill Shape=::NewEnumerator1
    Volume Origin:Vector3f=&AddVectorToPosition(Position=<-47631 21734 935>, Vector=<0 0 0>)
   ) {enabled=false, module=KillParticlesInVolume007}
   &KillParticlesInVolume(
    Box Size:Vector3f=<2350 2500 2625>
    Kill Shape=::NewEnumerator1
    Volume Origin:Vector3f=&AddVectorToPosition(Position=<-47604 24485 935>, Vector=<0 0 0>)
   ) {module=KillParticlesInVolume008}
   &KillParticlesInVolume(
    Box Size:Vector3f=<2350 2500 8000>
    Kill Shape=::NewEnumerator1
    Volume Origin:Vector3f=&AddVectorToPosition(Position=<-43503 24008 1359>, Vector=<0 0 0>)
   ) {module=KillParticlesInVolume009}
   &KillParticlesInVolume(
    Box Size:Vector3f=<2350 2475 2500>
    Kill Shape=::NewEnumerator1
    Volume Origin:Vector3f=&AddVectorToPosition(Position=<-55221 30183 1012>, Vector=<0 0 0>)
   ) {module=KillParticlesInVolume010}
   &KillParticlesInVolume(
    Box Size:Vector3f=<4925 4275 15000>
    Kill Shape=::NewEnumerator1
    Volume Origin:Vector3f=&AddVectorToPosition(Position=<-49721 29829 2402>, Vector=<0 0 0>)
   ) {module=KillParticlesInVolume003}
   &KillParticlesInVolume(
    Box Size:Vector3f=<2925 2475 10000>
    Kill Shape=::NewEnumerator1
    Volume Origin:Vector3f=&AddVectorToPosition(Position=<-42923 29166 1729>, Vector=<0 0 0>)
   ) {module=KillParticlesInVolume011}
   &KillParticlesInVolume(
    Box Size:Vector3f=<2925 3950 10000>
    Kill Shape=::NewEnumerator1
    Volume Origin:Vector3f=&AddVectorToPosition(Position=<-56620 38200 1729>, Vector=<0 0 0>)
   ) {module=KillParticlesInVolume012}
   &KillParticlesInVolume(
    Box Size:Vector3f=<2925 3950 10000>
    Kill Shape=::NewEnumerator1
    Volume Origin:Vector3f=&AddVectorToPosition(Position=<-51047 39987 2644>, Vector=<0 0 0>)
   ) {module=KillParticlesInVolume013}
   &KillParticlesInVolume(
    Box Size:Vector3f=<2300 2500 10000>
    Kill Shape=::NewEnumerator1
    Volume Origin:Vector3f=&AddVectorToPosition(Position=<-46708 40471 760>, Vector=<0 0 0>)
   ) {module=KillParticlesInVolume014}
   &KillParticlesInVolume(
    Box Size:Vector3f=<1100 1950 850>
    Kill Shape=::NewEnumerator1
    Volume Origin:Vector3f=&AddVectorToPosition(Position=<-47372 38200 -100>, Vector=<0 0 0>)
   ) {module=KillParticlesInVolume001}
   &self:RainTextureMasking(
    AlphaThreshold=0.5
    BoxCenter:NiagaraPosition=<-50828 29472 9916>
    BoxExtents:Vector3f=<32500 32500 1>
    Position:NiagaraPosition=$Particles.Position
    RejectionTexture=DataInterface() {propertyValues {
      Texture=/Game/Environments/Asteria/Sets/NeonCity/FX/Textures/T_NeonCity_Rain_Mask_01
      TextureUserParameter {parameter {name=None, typeDefHandle {registeredTypeIndex=2}, varData []}}
     }}
   ) {module=RainTextureMasking}
  ]
  particleUpdateScript [
   &ParticleState()
   &RegionMask(
    Inner Float=0
    Outer Float=2
    Region Falloff=0.1
    Region Radius:NiagaraFloat=$User.Rain_MaskInside_Radius
    UseOffset=false
    UseRotation=false
   ) {module=RegionMask001}
   &ScaleSpriteSize(Uniform Curve Sprite Scale=DataInterface() {propertyValues {Curve {keys [
        {arriveTangent=3.1525, interpMode=RCIM_Cubic, leaveTangent=3.1525, tangentMode=RCTM_User, time=0, value=0}
        {arriveTangent=-0.0405, interpMode=RCIM_Cubic, leaveTangent=-0.0405, tangentMode=RCTM_User, time=1, value=1}
       ]}}})
   &GravityForce(Gravity=<0 0 -150>)
   &DynamicMaterialParameters(
    Index 0 Param 1=&FloatFromCurve(
     FloatCurve=DataInterface() {propertyValues {Curve {keys [
         {arriveTangent=9.0725, interpMode=RCIM_Cubic, leaveTangent=9.0725, tangentMode=RCTM_User, time=0, value=0}
         {arriveTangent=-0.0783, interpMode=RCIM_Cubic, leaveTangent=-0.0783, tangentMode=RCTM_User, time=0.1408, value=0.9847}
         {arriveTangent=-0.0405, interpMode=RCIM_Cubic, leaveTangent=-0.0405, tangentMode=RCTM_User, time=1, value=0.5}
        ]}}}
     Scale Curve=2
    )
    Index 0 Param 2=0.1
    Index 0 Param 3=&RandomRangeFloat(Minimum=0, Maximum=0.5)
    Index 0 Param 4=1
    Module.Param0WriteEnabled=true
    Module.Param1WriteEnabled=false
    Module.Param2WriteEnabled=false
    Module.Param3WriteEnabled=false
   )
   &SolveForcesAndVelocity(Clamp Velocity=true, Limit Acceleration=false, Speed Limit=550)
   &ScaleColor(
    Scale Alpha=&Multiply_Float(
     A=&Multiply_Float(
      A=&Multiply_Float(
       A=&FloatFromCurve(FloatCurve=DataInterface() {propertyValues {Curve {keys [
            {interpMode=RCIM_Cubic, tangentMode=RCTM_User, time=0, value=0}
            {interpMode=RCIM_Cubic, tangentMode=RCTM_User, time=0.5, value=1}
            {tangentMode=RCTM_User, time=1, value=0}
           ]}}})
       B=$System.IndoorAlpha
      )
      B=$Particles.RegionMask001.Float
     )
     B=$User.Rain_AlphaMult
    )
    Scale RGB=&MakeVectorFromLinearColorRGB(Color=&MultiplyLinearColor(Linear Color=$User.Rain_NonLitColor, Scale Factor=&MakeVector4_FromFloat(Float=$User.Rain_ColorMult)))
    ScaleA=true
    ScaleRGB=true
   )
   /SetVariables(Particles.HighEPIC COLOR:LinearColor=&MakeLinearColorFromVectorAndFloat(
     "Vector (RGB)"=&Multiply_Vector(
      A=&MakeVectorFromLinearColorRGB(Color=$Particles.Initial.Color)
      B=&VectorFromFloat(Value=&Multiply_Float(A=10, B=$User.Rain_ColorMult))
     )
     "Float (Alpha)"=&Multiply_Float(
      A=&Multiply_Float(
       A=&Multiply_Float(
        A=&FloatFromCurve(FloatCurve=DataInterface() {propertyValues {Curve {keys [
             {interpMode=RCIM_Cubic, tangentMode=RCTM_User, time=0, value=0}
             {interpMode=RCIM_Cubic, tangentMode=RCTM_User, time=0.5, value=1}
             {tangentMode=RCTM_User, time=1, value=0}
            ]}}})
        B=$System.IndoorAlpha
       )
       B=$Particles.RegionMask001.Float
      )
      B=$User.Rain_AlphaMult
     )
    )) {isSetParameters=true}
  ]
  rendererClasses [&NiagaraSpriteRendererProperties]
  renderers [
   {rendererClass=&NiagaraSpriteRendererProperties}
   {rendererClass=&NiagaraSpriteRendererProperties, rendererIndex=1}
  ]
 }
}
summary {
 systemName=NS_Cave_Drips
 emitters [
  {emitterName=RainFromObjects, simTarget=GPUComputeSim, rendererClasses [&NiagaraSpriteRendererProperties]}
  {bEnabled=false, emitterName=SplashesClose, simTarget=GPUComputeSim, rendererClasses [&NiagaraSpriteRendererProperties]}
 ]
 userVariables [
  User.Gravity Scale {type=Vector3f, description="", defaultValue {struct=Vector3f, value=<0 0 -400>}}
  User.Radius Scale {type=NiagaraFloat, defaultValue=2500, description=""}
  User.Rain_AlphaMult {type=NiagaraFloat, defaultValue=0.075, description=""}
  User.Rain_ColorMult {type=NiagaraFloat, defaultValue=1, description=""}
  User.Rain_FallingSpeed {type=NiagaraFloat, defaultValue=1.5, description=""}
  User.Rain_InheritVel {type=Vector3f, description="", defaultValue {struct=Vector3f, value=<-1 -1 0>}}
  User.Rain_MaskInside_Radius {type=NiagaraFloat, defaultValue=500, description=""}
  User.Rain_NonLitColor {type=LinearColor, description="", defaultValue {struct=LinearColor, value=<rgba 0.4492 0.547 0.75 1>}}
  User.Rain_SpawnRateScale {type=NiagaraFloat, defaultValue=2, description=""}
  User.Spawn Rate Mult {type=NiagaraFloat, defaultValue=1, description=""}
 ]
}
